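# Modal Rejoin window shown for an existing game session, warning that abandoning may forfeit quest progress.
# A centered panel stacks a title header, description, automatic-rejoin countdown display and a footer with Retry, Rejoin and Abandon buttons bound to input actions.
# The event graph is empty; the texts and buttons are inherited bindings from native FortRejoinWindowBase.
# It declares OnRequestAbandon and OnRequestRejoinRetry dispatchers, with no calls to them here.
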WidgetBlueprint /Game/UI/JoinServer/RejoinWindow
graphs [
 EventGraph {links [], nodes []}
 OnRequestAbandon {nodes [OnRequestAbandon() {is=title K2Node_FunctionEntry}]}
 OnRequestRejoinRetry {nodes [OnRequestRejoinRetry() {is=title K2Node_FunctionEntry}]}
]
variables [
 OnRequestAbandon {}
 OnRequestRejoinRetry {}
]
widgets {
 info {inheritedWidgetCount=6, parentClass=/Script/FortniteUI.FortRejoinWindowBase, rootWidgetClass=&Overlay, widgetCount=20}
 widgets [
  Overlay_3 {widgetClassPath=&Overlay, widget=Overlay_3}
  SizeBox_0 {
   widgetClassPath=&SizeBox
   parent=Overlay_3
   widget=SizeBox_0
   properties {bOverride_MinDesiredHeight=true, bOverride_MinDesiredWidth=true, bOverride_WidthOverride=true, minDesiredHeight=600, widthOverride=1000}
   slot {class=OverlaySlot, properties {horizontalAlignment=HAlign_Center, verticalAlignment=VAlign_Center}}
  }
  CommonBorder_15 {
   widgetClassPath=&CommonBorder
   parent=SizeBox_0
   widget=CommonBorder_15
   properties {style=/FortUILibrary/Styles/Border/Border-MainModal.Border-MainModal_C, padding=<ltrb 0 0 0 0>}
   slot {class=SizeBoxSlot}
  }
  MainBox {
   widgetClassPath=&VerticalBox
   parent=CommonBorder_15
   widget=MainBox
   properties {navigation=&self:WidgetTree.MainBox.WidgetNavigation_0}
   slot {class=BorderSlot, properties {padding=<ltrb 0 0 0 0>}}
  }
  HorizontalBox_5 {widgetClassPath=&HorizontalBox, parent=MainBox, widget=HorizontalBox_5, slot {class=VerticalBoxSlot}}
  CommonBorder_Modal {
   widgetClassPath=&CommonBorder
   parent=HorizontalBox_5
   widget=CommonBorder_Modal
   properties {style=/FortUILibrary/Styles/Border/Border-ModalHeader.Border-ModalHeader_C}
   slot {class=HorizontalBoxSlot, properties {size {sizeRule=Fill, value=1}}}
  }
  HorizontalBox_1 {widgetClassPath=&HorizontalBox, parent=CommonBorder_Modal, widget=HorizontalBox_1, slot {class=BorderSlot}}
  Title {
   widgetClassPath=&CommonTextBlock
   bIsVariable=true
   parent=HorizontalBox_1
   widget=Title
   properties {
    style=/FortUILibrary/Styles/Typography/Foundation/TextStyle-Header-M.TextStyle-Header-M_C
    text=Rejoin
    font {
     bForceMonospaced=false
     bMaterialIsStencil=false
     fontMaterial=null
     fontObject=&Burbank
     letterSpacing=0
     monospacedWidth=1
     size=32
     skewAmount=0
     typefaceFontName=Burbank-Black
     outlineSettings {
      bApplyOutlineToDropShadows=false
      bMiteredCorners=false
      bSeparateFillAlpha=false
      outlineMaterial=null
      outlineSize=0
      outlineColor=<rgba 0 0 0 1>
     }
    }
    shadowOffset=<0 0>
   }
   slot {class=HorizontalBoxSlot, properties {horizontalAlignment=HAlign_Left, verticalAlignment=VAlign_Center, padding=<ltrb 16 16 16 16>}}
  }
  VerticalBox_0 {
   widgetClassPath=&VerticalBox
   parent=MainBox
   widget=VerticalBox_0
   slot {class=VerticalBoxSlot, properties {size {sizeRule=Fill, value=1}}}
  }
  Text_Description {
   widgetClassPath=&CommonTextBlock
   bInherited=true
   bIsVariable=true
   parent=VerticalBox_0
   widget=Text_Description
   properties {
    autoWrapText=true
    style=/FortUILibrary/Styles/Typography/Foundation/TextStyle-Base-S-B-Blue.TextStyle-Base-S-B-Blue_C
    text="An existing game session was found. Abandoning the game may forfeit quest progress. "
    colorAndOpacity {colorUseRule=UseColor_Specified, specifiedColor=<rgba 0.008 0.011 0.0262 1>}
    font {
     bForceMonospaced=false
     bMaterialIsStencil=false
     fontMaterial=null
     fontObject=&NotoSans
     letterSpacing=0
     monospacedWidth=1
     size=18
     skewAmount=0
     typefaceFontName=NotoSans-Bold
     outlineSettings {
      bApplyOutlineToDropShadows=false
      bMiteredCorners=false
      bSeparateFillAlpha=false
      outlineMaterial=null
      outlineSize=0
      outlineColor=<rgba 0 0 0 1>
     }
    }
    shadowOffset=<0 0>
   }
   slot {class=VerticalBoxSlot, properties {padding=<ltrb 32 32 32 32>}}
  }
  CommonBorder_1 {
   widgetClassPath=&CommonBorder
   parent=VerticalBox_0
   widget=CommonBorder_1
   properties {
    background {
     bIsDynamicallyLoaded=false
     drawAs=Border
     imageType=NoImage
     mirroring=NoMirror
     resourceName=None
     resourceObject=/Game/Effects/SplineTrails/Textures/T_LinearColorCurve_Default_01
     tiling=NoTile
     imageSize=<4 4>
     margin=<ltrb 0.5 0.5 0.5 0.5>
     outlineSettings {
      bUseBrushTransparency=false
      roundingType=HalfHeightRadius
      width=0
      color {colorUseRule=UseColor_Specified, specifiedColor=<rgba 0 0 0 0>}
      cornerRadii=<0 0 0 0>
     }
     tintColor {colorUseRule=UseColor_Specified, specifiedColor=<rgba 1 1 1 1>}
     uVRegion {bIsValid=false, max=<0 0>, min=<0 0>}
    }
    brushColor=<rgba 0.6146 0.6146 0.6146 1>
    padding=<ltrb 1 1 1 1>
   }
   slot {class=VerticalBoxSlot, properties {verticalAlignment=VAlign_Center, padding=<ltrb 32 16 32 32>}}
  }
  CommonBorder_0 {
   widgetClassPath=&CommonBorder
   parent=CommonBorder_1
   widget=CommonBorder_0
   properties {style=/FortUILibrary/Styles/Border/Border-TabM-Solid-White60pc.Border-TabM-Solid-White60pc_C, padding=<ltrb 16 16 16 16>}
   slot {class=BorderSlot, properties {padding=<ltrb 1 1 1 1>}}
  }
  VerticalBox_1 {
   widgetClassPath=&VerticalBox
   parent=CommonBorder_0
   widget=VerticalBox_1
   slot {class=BorderSlot, properties {padding=<ltrb 16 16 16 16>}}
  }
  Text_AutoRejoin {
   widgetClassPath=&CommonTextBlock
   bInherited=true
   bIsVariable=true
   parent=VerticalBox_1
   widget=Text_AutoRejoin
   properties {
    justification=Center
    style=/FortUILibrary/Styles/Typography/Foundation/TextStyle-Base-S-B-Gray.TextStyle-Base-S-B-Gray_C
    text=AUTOMATIC REJOIN IN:
    colorAndOpacity {colorUseRule=UseColor_Specified, specifiedColor=<rgba 0.0976 0.0976 0.0976 1>}
    font {
     bForceMonospaced=false
     bMaterialIsStencil=false
     fontMaterial=null
     fontObject=&NotoSans
     letterSpacing=0
     monospacedWidth=1
     size=18
     skewAmount=0
     typefaceFontName=NotoSans-Bold
     outlineSettings {
      bApplyOutlineToDropShadows=false
      bMiteredCorners=false
      bSeparateFillAlpha=false
      outlineMaterial=null
      outlineSize=0
      outlineColor=<rgba 0 0 0 1>
     }
    }
    shadowOffset=<0 0>
   }
   slot {class=VerticalBoxSlot, properties {horizontalAlignment=HAlign_Center, verticalAlignment=VAlign_Center}}
  }
  Text_RejoinTime {
   widgetClassPath=&CommonTextBlock
   bInherited=true
   bIsVariable=true
   parent=VerticalBox_1
   widget=Text_RejoinTime
   properties {
    justification=Center
    style=/FortUILibrary/Styles/Typography/Foundation/TextStyle-Base-S-B-Blue-40.TextStyle-Base-S-B-Blue-40_C
    text=0:00
    colorAndOpacity {colorUseRule=UseColor_Specified, specifiedColor=<rgba 0.008 0.011 0.0262 1>}
    font {
     bForceMonospaced=false
     bMaterialIsStencil=false
     fontMaterial=null
     fontObject=&NotoSans
     letterSpacing=0
     monospacedWidth=1
     size=18
     skewAmount=0
     typefaceFontName=NotoSans-Bold
     outlineSettings {
      bApplyOutlineToDropShadows=false
      bMiteredCorners=false
      bSeparateFillAlpha=false
      outlineMaterial=null
      outlineSize=0
      outlineColor=<rgba 0 0 0 1>
     }
    }
    shadowOffset=<0 0>
   }
   slot {class=VerticalBoxSlot, properties {horizontalAlignment=HAlign_Center, verticalAlignment=VAlign_Center, padding=<ltrb 0 2 0 0>}}
  }
  CommonBorder_2 {
   widgetClassPath=&CommonBorder
   parent=MainBox
   widget=CommonBorder_2
   properties {
    horizontalAlignment=HAlign_Right
    style=/FortUILibrary/Styles/Border/Border_LightNavy_VGrad.Border_LightNavy_VGrad_C
    padding=<ltrb 16 16 16 16>
   }
   slot {class=VerticalBoxSlot, properties {verticalAlignment=VAlign_Bottom}}
  }
  ButtonBox {
   widgetClassPath=&HorizontalBox
   bIsVariable=true
   parent=CommonBorder_2
   widget=ButtonBox
   slot {class=BorderSlot, properties {horizontalAlignment=HAlign_Right, padding=<ltrb 16 16 16 16>}}
  }
  Button_Retry {
   widgetClassPath=&IconTextButton
   bInherited=true
   bIsVariable=true
   parent=ButtonBox
   widget=Button_Retry
   properties {
    onButtonBaseClicked=(&self:WidgetTree.Button_Retry.HandleOnButtonClicked)
    onButtonBaseDoubleClicked=(&self:WidgetTree.Button_Retry.HandleOnButtonDoubleClicked)
    onButtonBaseHovered=(&self:WidgetTree.Button_Retry.HandleOnButtonHovered)
    onButtonBaseUnhovered=(&self:WidgetTree.Button_Retry.HandleOnButtonUnhovered)
    onSelectedChangedBase=(&self:WidgetTree.Button_Retry.HandleOnSelectedChanged)
    triggeringInputAction {dataTable=&InputActionDataTable, rowName=Retry}
   }
   slot {class=HorizontalBoxSlot, properties {padding=<ltrb 10 0 10 0>}}
  }
  Button_Rejoin {
   widgetClassPath=&IconTextButton
   bInherited=true
   bIsVariable=true
   parent=ButtonBox
   widget=Button_Rejoin
   properties {
    onButtonBaseClicked=(&self:WidgetTree.Button_Rejoin.HandleOnButtonClicked)
    onButtonBaseDoubleClicked=(&self:WidgetTree.Button_Rejoin.HandleOnButtonDoubleClicked)
    onButtonBaseHovered=(&self:WidgetTree.Button_Rejoin.HandleOnButtonHovered)
    onButtonBaseUnhovered=(&self:WidgetTree.Button_Rejoin.HandleOnButtonUnhovered)
    onSelectedChangedBase=(&self:WidgetTree.Button_Rejoin.HandleOnSelectedChanged)
    triggeringInputAction {dataTable=&InputActionDataTable, rowName=RejoinGame}
   }
   slot {class=HorizontalBoxSlot, properties {padding=<ltrb 10 0 10 0>}}
  }
  Button_Abandon {
   widgetClassPath=&IconTextButton
   bInherited=true
   bIsVariable=true
   parent=ButtonBox
   widget=Button_Abandon
   properties {
    onButtonBaseClicked=(&self:WidgetTree.Button_Abandon.HandleOnButtonClicked)
    onButtonBaseDoubleClicked=(&self:WidgetTree.Button_Abandon.HandleOnButtonDoubleClicked)
    onButtonBaseHovered=(&self:WidgetTree.Button_Abandon.HandleOnButtonHovered)
    onButtonBaseUnhovered=(&self:WidgetTree.Button_Abandon.HandleOnButtonUnhovered)
    onSelectedChangedBase=(&self:WidgetTree.Button_Abandon.HandleOnSelectedChanged)
    triggeringInputAction {dataTable=&InputActionDataTable, rowName=AbandonGame}
   }
   slot {class=HorizontalBoxSlot, properties {padding=<ltrb 10 0 10 0>}}
  }
 ]
}
